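# Physics asset for the hair dynamics of the AnvilAlarmRobin_Extra male medium character's FaceAcc mesh part.
# Seven two-segment capsule chains (front left/right, side left/right, back left/center/right) each hang from a small base hair body.
# Constraints link base to first to second segment, with twist locked and swing angles varying by chain.
# Larger head, neck, clavicle and upper-spine capsules approximate the body around the hair and carry no constraints.
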PhysicsAsset /BRCosmetics/Characters/Player/Male/Medium/Bodies/M_MED_AnvilAlarmRobin_Extra/Meshes/Parts/M_MED_AnvilAlarmRobin_Extra_FaceAcc_Physics
bodies {
 base_hair_bk_c [{shapeType=Capsule, length=0.505, radius=0.505, shapeName=base_hair_bk_c_capsule, rotation=<rot 0 -90 0>}]
 base_hair_bk_l [{shapeType=Capsule, length=0.505, radius=0.505, shapeName=base_hair_bk_l_capsule, rotation=<rot 0 -90 0>}]
 base_hair_bk_r [{shapeType=Capsule, length=0.505, radius=0.505, shapeName=base_hair_bk_r_capsule, rotation=<rot 0 -90 0>}]
 base_hair_fr_l [{shapeType=Capsule, length=0.505, radius=0.505, shapeName=base_hair_fr_l_capsule, rotation=<rot 0 -90 0>}]
 base_hair_fr_r [{shapeType=Capsule, length=0.505, radius=0.505, shapeName=base_hair_fr_r_capsule, rotation=<rot 0 -90 0>}]
 base_hair_sd_l [{shapeType=Capsule, length=0.505, radius=0.505, shapeName=base_hair_sd_l_capsule, rotation=<rot 0 -90 0>}]
 base_hair_sd_r [{shapeType=Capsule, length=0.505, radius=0.505, shapeName=base_hair_sd_r_capsule, rotation=<rot 0 -90 0>}]
 clavicle_extra_l [{
   shapeType=Capsule
   length=11.6233
   radius=4.9352
   shapeName=clavicle_l_capsule
   center=<15.8566 -2.9379 -0.4268>
   rotation=<rot 83.1122 -0.0006 179.9994>
  }]
 clavicle_extra_r [{
   shapeType=Capsule
   length=10.6705
   radius=6.7
   shapeName=clavicle_r_capsule
   center=<-15.7861 0 1.0108>
   rotation=<rot 75.2375 -0.0006 -0.0006>
  }]
 dyn_hair_bk_01_c [{shapeType=Capsule, length=3.2584, radius=2, shapeName=dyn_hair_bk_01_c_capsule, center=<2.7786 0 1.1444>, rotation=<rot 90 90 0>}]
 dyn_hair_bk_01_l [{shapeType=Capsule, length=4.9258, radius=2, shapeName=dyn_hair_bk_01_l_capsule, center=<2.7786 0 1.3351>, rotation=<rot 90 90 0>}]
 dyn_hair_bk_01_r [{shapeType=Capsule, length=4.9258, radius=2, shapeName=dyn_hair_bk_01_r_capsule, center=<-2.7786 0 -1.3351>, rotation=<rot 90 -90 0>}]
 dyn_hair_bk_02_c [{shapeType=Capsule, length=2.0679, radius=2, shapeName=dyn_hair_bk_02_c_capsule, center=<5.4092 0 0.6425>, rotation=<rot 90 90 0>}]
 dyn_hair_bk_02_l [{shapeType=Capsule, length=4.4274, radius=2, shapeName=dyn_hair_bk_02_l_capsule, center=<4.5628 0 1.182>, rotation=<rot 90 90 0>}]
 dyn_hair_bk_02_r [{
   shapeType=Capsule
   length=5.0488
   radius=2
   shapeName=dyn_hair_bk_02_r_capsule
   center=<-4.4131 0.2644 -0.6622>
   rotation=<rot 81.7752 -89.999 0.0009>
  }]
 dyn_hair_fr_01_l [{
   shapeType=Capsule
   length=4.65
   radius=1.35
   shapeName=dyn_hair_fr_01_l_capsule
   center=<2.3482 -0.3418 0.8263>
   rotation=<rot 84.2385 90.0003 0.0002>
  }]
 dyn_hair_fr_01_r [{
   shapeType=Capsule
   length=4.65
   radius=1.35
   shapeName=dyn_hair_fr_01_r_capsule
   center=<-2.3482 0.3418 -0.8263>
   rotation=<rot 84.2385 90.0003 -179.9998>
  }]
 dyn_hair_fr_02_l [{
   shapeType=Capsule
   length=4.2397
   radius=1.7603
   shapeName=dyn_hair_fr_02_l_capsule
   center=<3.9298 -1.0651 0.7823>
   rotation=<rot 84.2385 90.0003 0.0002>
  }]
 dyn_hair_fr_02_r [{
   shapeType=Capsule
   length=5.9885
   radius=1.7603
   shapeName=dyn_hair_fr_02_r_capsule
   center=<-5.0404 1.9163 -1.0436>
   rotation=<rot 84.2385 90.0003 -179.9998>
  }]
 dyn_hair_sd_01_l [{
   shapeType=Capsule
   length=4.55
   radius=1.45
   shapeName=dyn_hair_sd_01_l_capsule
   center=<3.6871 0.0233 0.5761>
   rotation=<rot 81.7592 89.9581 0>
  }]
 dyn_hair_sd_01_r [{
   shapeType=Capsule
   length=4.55
   radius=1.45
   shapeName=dyn_hair_sd_01_r_capsule
   center=<-3.6871 -0.0233 -0.5761>
   rotation=<rot 81.7592 89.9581 180>
  }]
 dyn_hair_sd_02_l [{shapeType=Capsule, length=4.02, radius=1.98, shapeName=dyn_hair_sd_02_l_capsule, center=<6.2754 0 1.2279>, rotation=<rot 90 90 0>}]
 dyn_hair_sd_02_r [{
   shapeType=Capsule
   length=4.02
   radius=1.98
   shapeName=dyn_hair_sd_02_r_capsule
   center=<-5.1825 0.1352 -2.3771>
   rotation=<rot 80.2583 -89.9999 0>
  }]
 head [{shapeType=Capsule, length=5.9421, radius=6.6468, shapeName=head_capsule, center=<2.2093 1.3029 0>, rotation=<rot 90 0 0>}]
 neck_01 [{shapeType=Capsule, length=3.5584, radius=6.4416, shapeName=neck_01_capsule, center=<0.3436 0.942 0>, rotation=<rot 90 -5.6783 0>}]
 neck_02 [{shapeType=Capsule, length=6.6189, radius=6.4392, shapeName=neck_02_capsule, center=<0.7614 -0.1533 0>, rotation=<rot 90 -11.3816 0>}]
 spine_05 [
  {shapeType=Capsule, length=15.1979, radius=15, shapeName=spine_05_capsule, center=<-0.8587 2.8293 0>, rotation=<rot 0 -90 0>}
  {shapeType=Capsule, length=8.6654, radius=7, shapeName=spine_05_capsule, center=<9.5662 2.7481 0>}
 ]
}
constraints [
 {bone1Name=dyn_hair_fr_01_l, bone2Name=base_hair_fr_l, swing1LimitDegrees=15, swing2LimitDegrees=20, twistLimitDegrees=45, twistMotion=Locked}
 {
  bone1Name=dyn_hair_fr_02_l
  bone2Name=dyn_hair_fr_01_l
  swing1LimitDegrees=15
  swing2LimitDegrees=20
  twistLimitDegrees=45
  twistMotion=Locked
 }
 {bone1Name=dyn_hair_fr_01_r, bone2Name=base_hair_fr_r, swing1LimitDegrees=15, swing2LimitDegrees=20, twistLimitDegrees=45, twistMotion=Locked}
 {
  bone1Name=dyn_hair_fr_02_r
  bone2Name=dyn_hair_fr_01_r
  swing1LimitDegrees=15
  swing2LimitDegrees=20
  twistLimitDegrees=45
  twistMotion=Locked
 }
 {bone1Name=dyn_hair_sd_01_l, bone2Name=base_hair_sd_l, swing1LimitDegrees=20, swing2LimitDegrees=20, twistLimitDegrees=45, twistMotion=Locked}
 {
  bone1Name=dyn_hair_sd_02_l
  bone2Name=dyn_hair_sd_01_l
  swing1LimitDegrees=20
  swing2LimitDegrees=20
  twistLimitDegrees=45
  twistMotion=Locked
 }
 {bone1Name=dyn_hair_sd_01_r, bone2Name=base_hair_sd_r, swing1LimitDegrees=20, swing2LimitDegrees=20, twistLimitDegrees=45, twistMotion=Locked}
 {
  bone1Name=dyn_hair_sd_02_r
  bone2Name=dyn_hair_sd_01_r
  swing1LimitDegrees=20
  swing2LimitDegrees=20
  twistLimitDegrees=45
  twistMotion=Locked
 }
 {bone1Name=dyn_hair_bk_01_r, bone2Name=base_hair_bk_r, swing1LimitDegrees=10, swing2LimitDegrees=30, twistLimitDegrees=15, twistMotion=Locked}
 {
  bone1Name=dyn_hair_bk_02_r
  bone2Name=dyn_hair_bk_01_r
  swing1LimitDegrees=10
  swing2LimitDegrees=30
  twistLimitDegrees=15
  twistMotion=Locked
 }
 {
  bone1Name=dyn_hair_bk_02_l
  bone2Name=dyn_hair_bk_01_l
  swing1LimitDegrees=20
  swing2LimitDegrees=20
  twistLimitDegrees=45
  twistMotion=Locked
 }
 {bone1Name=dyn_hair_bk_01_l, bone2Name=base_hair_bk_l, swing1LimitDegrees=10, swing2LimitDegrees=20, twistLimitDegrees=45, twistMotion=Locked}
 {
  bone1Name=dyn_hair_bk_02_c
  bone2Name=dyn_hair_bk_01_c
  swing1LimitDegrees=10
  swing2LimitDegrees=20
  twistLimitDegrees=45
  twistMotion=Locked
 }
 {bone1Name=dyn_hair_bk_01_c, bone2Name=base_hair_bk_c, swing1LimitDegrees=10, swing2LimitDegrees=20, twistLimitDegrees=45, twistMotion=Locked}
]
