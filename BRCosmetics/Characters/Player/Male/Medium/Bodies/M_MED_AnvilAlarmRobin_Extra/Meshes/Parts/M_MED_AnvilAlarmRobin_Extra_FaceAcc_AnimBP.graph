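# Anim Blueprint for the AnvilAlarmRobin Extra face accessory, copying the attached parent's pose through a Control Rig fed by movement, emote, front-end idle and hang-glider state.
# The cached result gets secondary physics: chained AnimDynamics on hair bones when an AnvilAlarmMoth part mesh name is detected.
# Otherwise RigidBody simulates relative to neck_02 during emotes, or to head with an additive head_back tilt.
# The event graph also computes gravity-override and RBAN settings.
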AnimBlueprint /BRCosmetics/Characters/Player/Male/Medium/Bodies/M_MED_AnvilAlarmRobin_Extra/Meshes/Parts/M_MED_AnvilAlarmRobin_Extra_FaceAcc_AnimBP
target_skeleton=&/M_MED_AnvilAlarmRobin_Extra_FaceAcc_Skeleton
graphs [
 AnimGraph {nodes [
   n1 = RigidBody!(
    ComponentPose=Local To Component(LocalPose=Use cached pose 'CtrlRig Cache'())
    Alpha
    OverrideWorldGravity:Vector="0,0,-980"
    ExternalForce
    SimSpaceSettings:Sim Space Settings Structure=(WorldAlpha=0.7,VelocityScaleZ=1,DampingAlpha=1,MaxLinearVelocity=200,MaxAngularVelocity=20,MaxLinearAcceleration=400,MaxAngularAcceleration=100,ExternalLinearDragV=(X=0,Y=0,Z=0),ExternalLinearVelocity=(X=0,Y=0,Z=0),ExternalAngularVelocity=(X=0,Y=0,Z=0))
   ) -> (Pose) {
    bOverrideWorldGravity=true
    overridePhysicsAsset=&/M_MED_AnvilAlarmRobin_Extra_FaceAcc_Physics
    simulationSpace=BaseBoneSpace
    alphaScaleBiasClamp {bInterpResult=true}
    baseBoneRef {boneName=neck_02}
    componentAppliedLinearAccClamp=<0 0 0>
    overrideWorldGravity {z=-980}
    simSpaceSettings {maxAngularAcceleration=100, maxAngularVelocity=20, maxLinearAcceleration=400, maxLinearVelocity=200, worldAlpha=0.7}
   }
   n2 = Component To Local(ComponentPose=@n1)
   n3 = RigidBody(
    ComponentPose=Local To Component(LocalPose=Use cached pose 'CtrlRig Cache'())
    OverrideWorldGravity="0,0,-980"
    SimSpaceSettings=(WorldAlpha=0.7,VelocityScaleZ=1,DampingAlpha=1,MaxLinearVelocity=200,MaxAngularVelocity=20,MaxLinearAcceleration=400,MaxAngularAcceleration=100,ExternalLinearDragV=(X=0,Y=0,Z=0),ExternalLinearVelocity=(X=0,Y=0,Z=0),ExternalAngularVelocity=(X=0,Y=0,Z=0))
   ) {
    bOverrideWorldGravity=true
    overridePhysicsAsset=&M_MED_AnvilAlarmRobin_Extra_FaceAcc_Physics
    simulationSpace=BaseBoneSpace
    alphaScaleBiasClamp {bInterpResult=true}
    baseBoneRef {boneName=head}
    componentAppliedLinearAccClamp=<0 0 0>
    overrideWorldGravity {z=-980}
    simSpaceSettings {maxAngularAcceleration=100, maxAngularVelocity=20, maxLinearAcceleration=400, maxLinearVelocity=200, worldAlpha=0.7}
   }
   n4 = "Transform (Modify) Bone - Bone: head_back"!(ComponentPose=@n3, bAlphaBoolEnabled:bool=True, Translation="0,0,0", Rotation="-27,0,0", Scale="1,1,1") -> (Pose) {
    alphaInputType=Bool
    rotationMode=BMM_Additive
    rotationSpace=BCS_BoneSpace
    alphaBoolBlend {blendInTime=0.6, blendOutTime=0.1}
    boneToModify {boneName=head_back}
    rotation {pitch=-27}
   }
   n5 = Component To Local(ComponentPose=@n4)
   n6 = Get bIsPlayingEmote()
   n7 = Anim Dynamics - Bone: dyn_hair_fr_01_l!(
    ComponentPose=Local To Component(LocalPose=Use cached pose 'CtrlRig Cache'())
    Alpha=0.5
    GravityOverride:Vector="0,0,-980"
    ExternalForce="0,0,0"
   ) -> (Pose) {
    angularDampingOverride=0.9999
    angularSpringConstant=10000
    bAngularSpring=true
    bOverrideAngularDamping=true
    bUseGravityOverride=true
    gravityScale=0.35
    simulationSpace=RootRelative
    boundBone {boneName=dyn_hair_fr_01_l}
    componentAppliedLinearAccClamp=<2000 2000 2000>
    componentLinearAccScale=<1 1 1>
    componentLinearVelScale=<1 1 1>
    gravityOverride {z=-980}
    physicsBodyDefinitions [{
      boundBone {boneName=dyn_hair_fr_01_l}
      boxExtents=<20 30 30>
      constraintSetup {angularLimitsMax=<5 0 15>, angularLimitsMin=<-5 0 -5>, angularTarget=<1 0 0>}
      localJointOffset=<4 0 0>
     }]
    simSpaceSettings {maxAngularAcceleration=100, maxAngularVelocity=50, simSpaceAngularAlpha=0.5}
   }
   n8 = Anim Dynamics - Bone: dyn_hair_fr_02_l!(ComponentPose=@n7, Alpha=0.5, GravityOverride:Vector="0,0,-980", ExternalForce="0,0,0") -> (Pose) {
    angularDampingOverride=0.9999
    angularSpringConstant=10000
    bAngularSpring=true
    bOverrideAngularDamping=true
    bUseGravityOverride=true
    gravityScale=0.35
    simulationSpace=RootRelative
    boundBone {boneName=dyn_hair_fr_02_l}
    componentAppliedLinearAccClamp=<2000 2000 2000>
    componentLinearAccScale=<1 1 1>
    componentLinearVelScale=<1 1 1>
    gravityOverride {z=-980}
    physicsBodyDefinitions [{
      boundBone {boneName=dyn_hair_fr_02_l}
      boxExtents=<20 30 30>
      constraintSetup {angularLimitsMax=<5 0 15>, angularLimitsMin=<-5 0 -5>, angularTarget=<1 0 0>}
      localJointOffset=<4 0 0>
     }]
    simSpaceSettings {maxAngularAcceleration=100, maxAngularVelocity=50, simSpaceAngularAlpha=0.5}
   }
   n9 = Anim Dynamics - Bone: dyn_hair_sd_01_l!(ComponentPose=@n8, Alpha=0.5, GravityOverride:Vector="0,0,-980", ExternalForce="0,0,0") -> (Pose) {
    angularDampingOverride=0.9999
    angularSpringConstant=10000
    bAngularSpring=true
    bOverrideAngularDamping=true
    bUseGravityOverride=true
    gravityScale=0.35
    simulationSpace=RootRelative
    boundBone {boneName=dyn_hair_sd_01_l}
    componentAppliedLinearAccClamp=<2000 2000 2000>
    componentLinearAccScale=<1 1 1>
    componentLinearVelScale=<1 1 1>
    gravityOverride {z=-980}
    physicsBodyDefinitions [{
      boundBone {boneName=dyn_hair_sd_01_l}
      boxExtents=<20 30 30>
      constraintSetup {angularLimitsMax=<5 0 15>, angularLimitsMin=<-5 0 -5>, angularTarget=<1 0 0>}
      localJointOffset=<4 0 0>
     }]
    simSpaceSettings {maxAngularAcceleration=100, maxAngularVelocity=50, simSpaceAngularAlpha=0.5}
   }
   n10 = Anim Dynamics - Bone: dyn_hair_sd_02_l!(ComponentPose=@n9, Alpha=0.5, GravityOverride:Vector="0,0,-980", ExternalForce="0,0,0") -> (Pose) {
    angularDampingOverride=0.9999
    angularSpringConstant=10000
    bAngularSpring=true
    bOverrideAngularDamping=true
    bUseGravityOverride=true
    gravityScale=0.35
    simulationSpace=RootRelative
    boundBone {boneName=dyn_hair_sd_02_l}
    componentAppliedLinearAccClamp=<2000 2000 2000>
    componentLinearAccScale=<1 1 1>
    componentLinearVelScale=<1 1 1>
    gravityOverride {z=-980}
    physicsBodyDefinitions [{
      boundBone {boneName=dyn_hair_sd_02_l}
      boxExtents=<20 30 30>
      constraintSetup {angularLimitsMax=<5 0 15>, angularLimitsMin=<-5 0 -5>, angularTarget=<1 0 0>}
      localJointOffset=<4 0 0>
     }]
    simSpaceSettings {maxAngularAcceleration=100, maxAngularVelocity=50, simSpaceAngularAlpha=0.5}
   }
   n11 = Anim Dynamics - Bone: dyn_hair_fr_01_r!(ComponentPose=@n10, Alpha=0.5, GravityOverride:Vector="0,0,-980", ExternalForce="0,0,0") -> (Pose) {
    angularDampingOverride=0.9999
    angularSpringConstant=10000
    bAngularSpring=true
    bOverrideAngularDamping=true
    bUseGravityOverride=true
    gravityScale=0.35
    simulationSpace=RootRelative
    boundBone {boneName=dyn_hair_fr_01_r}
    componentAppliedLinearAccClamp=<2000 2000 2000>
    componentLinearAccScale=<1 1 1>
    componentLinearVelScale=<1 1 1>
    gravityOverride {z=-980}
    physicsBodyDefinitions [{
      boundBone {boneName=dyn_hair_fr_01_r}
      boxExtents=<20 30 30>
      constraintSetup {angularLimitsMax=<5 0 15>, angularLimitsMin=<-5 0 -5>, angularTarget=<1 0 0>}
      localJointOffset=<-4 0 0>
     }]
    simSpaceSettings {maxAngularAcceleration=100, maxAngularVelocity=50, simSpaceAngularAlpha=0.5}
   }
   n12 = Anim Dynamics - Bone: dyn_hair_fr_02_r!(ComponentPose=@n11, Alpha=0.5, GravityOverride:Vector="0,0,-980", ExternalForce="0,0,0") -> (Pose) {
    angularDampingOverride=0.9999
    angularSpringConstant=10000
    bAngularSpring=true
    bOverrideAngularDamping=true
    bUseGravityOverride=true
    gravityScale=0.35
    simulationSpace=RootRelative
    boundBone {boneName=dyn_hair_fr_02_r}
    componentAppliedLinearAccClamp=<2000 2000 2000>
    componentLinearAccScale=<1 1 1>
    componentLinearVelScale=<1 1 1>
    gravityOverride {z=-980}
    physicsBodyDefinitions [{
      boundBone {boneName=dyn_hair_fr_02_r}
      boxExtents=<20 30 30>
      constraintSetup {angularLimitsMax=<5 0 15>, angularLimitsMin=<-5 0 -5>, angularTarget=<1 0 0>}
      localJointOffset=<-4 0 0>
     }]
    simSpaceSettings {maxAngularAcceleration=100, maxAngularVelocity=50, simSpaceAngularAlpha=0.5}
   }
   n13 = Anim Dynamics - Bone: dyn_hair_sd_01_r!(ComponentPose=@n12, Alpha=0.5, GravityOverride:Vector="0,0,-980", ExternalForce="0,0,0") -> (Pose) {
    angularDampingOverride=0.9999
    angularSpringConstant=20000
    bAngularSpring=true
    bOverrideAngularDamping=true
    bUseGravityOverride=true
    gravityScale=0.35
    simulationSpace=RootRelative
    boundBone {boneName=dyn_hair_sd_01_r}
    componentAppliedLinearAccClamp=<2000 2000 2000>
    componentLinearAccScale=<1 1 1>
    componentLinearVelScale=<1 1 1>
    gravityOverride {z=-980}
    physicsBodyDefinitions [{
      boundBone {boneName=dyn_hair_sd_01_r}
      boxExtents=<20 30 30>
      constraintSetup {angularLimitsMax=<5 0 15>, angularLimitsMin=<-5 0 -5>, angularTarget=<1 0 0>}
      localJointOffset=<-4 0 0>
     }]
    simSpaceSettings {maxAngularAcceleration=100, maxAngularVelocity=50, simSpaceAngularAlpha=0.5}
   }
   n14 = Anim Dynamics - Bone: dyn_hair_sd_02_r!(ComponentPose=@n13, Alpha=0.5, GravityOverride:Vector="0,0,-980", ExternalForce="0,0,0") -> (Pose) {
    angularDampingOverride=0.9999
    angularSpringConstant=20000
    bAngularSpring=true
    bOverrideAngularDamping=true
    bUseGravityOverride=true
    gravityScale=0.35
    simulationSpace=RootRelative
    boundBone {boneName=dyn_hair_sd_02_r}
    componentAppliedLinearAccClamp=<2000 2000 2000>
    componentLinearAccScale=<1 1 1>
    componentLinearVelScale=<1 1 1>
    gravityOverride {z=-980}
    physicsBodyDefinitions [{
      boundBone {boneName=dyn_hair_sd_02_r}
      boxExtents=<20 30 30>
      constraintSetup {angularLimitsMax=<5 0 15>, angularLimitsMin=<-5 0 -5>, angularTarget=<1 0 0>}
      localJointOffset=<-4 0 0>
     }]
    simSpaceSettings {maxAngularAcceleration=100, maxAngularVelocity=50, simSpaceAngularAlpha=0.5}
   }
   Output Pose(Result=Blend Poses by bool(
     BlendPose_0=Component To Local(ComponentPose=@n14)
     BlendPose_1=Blend Poses by bool(BlendPose_0=@n2, BlendPose_1=@n5, bActiveValue=@n6)
     bActiveValue=Get AnvilAlarmMoth Detected() -> (AnvilAlarmMoth Detected:bool)
    )) {layerGroup=None}
   n15 = Control Rig!(
    Source=Copy Pose From Mesh() {bUseAttachedParent=true}
    Alpha
    Crouching:bool=Get bIsCrouching()
    Moving2D:bool=Get bIsMoving2D()
    Targeting:bool=Get bIsTargeting()
    Swimming:bool=Get bIsSurfaceSwimming()
    Emote:bool=Get bIsPlayingEmote()
    RBAN Enable:bool=Get bRigidBodyEnabled()
    DetectedAnvilAlarmMoth:bool=Get AnvilAlarmMoth Detected()
    FEIdle Cons Value:double=Get FE Idle Cons Value() -> (FE Idle Cons Value:double)
    HangGliding:double=Get HangGlidingAlpha() -> (HangGlidingAlpha:double)
   ) -> (Pose) {controlRigAssetReference {blueprintRigClass=&/M_MED_AnvilAlarmRobin_Extra_FaceAcc_CtrlRig.M_MED_AnvilAlarmRobin_Extra_FaceAcc_CtrlRig_C}}
   Save cached pose 'CtrlRig Cache'(Pose=@n15) {cacheName=CtrlRig Cache}
  ]}
 EventGraph {nodes [
   Event BlueprintUpdateAnimation()
   > n1 = Sequence!(execute) -> (then_0, then_1, then_2:Exec)
   > n2 = "Gravity Override State Function (Parameter Asset) (Nativized)"(
    CharacterPartAnimInstance=Self-Reference()
    GameTimeInSeconds=GetGameTimeInSeconds()
    DebugJointName=spine_05
    Parameters=&/M_MED_AnvilAlarmRobin_Extra_FaceAcc_Parameters
   )
   > Set Gravity Override(Gravity Override=@n2)
   > Set RBAN Settings Out(RBAN Settings Out=@n2.OutRigidBodyAnimNodeParameters)
   n3 = Get bIsSkyDiving()
   n4 = Get bIsMoving2D()
   n5 = Get bIsPlayingEmote()
   n6 = Branch(
    execute=@n1.then_2
    Condition=AND Boolean!(A=Get bIsInFrontEnd(), B=NOT Boolean(A=@n4), C:bool=NOT Boolean(A=@n5), D:bool=NOT Boolean(A=@n3)) -> (ReturnValue)
   )
   > Set FE Idle(FE Idle:double=1) -> (Output_Get:double)
   > Set FE Idle Cons Value(FE Idle Cons Value:double=0) -> (Output_Get:double)
   Set FE Idle(execute=@n6.else, FE Idle=0.5)
   > Set FE Idle Cons Value(FE Idle Cons Value=0.6)
   n7 = Cast To FortPlayerPawn(execute=@n1.then_1, Object=TryGetPawnOwner())
   n8 = GetSkeletalMeshForPartType(self=@n7)
   > Set AnvilAlarmMoth Detected(AnvilAlarmMoth Detected:bool=MatchesWildcard(SourceString=GetDisplayName(Object=@n8), Wildcard=*AnvilAlarmMoth*, SearchCase=CaseSensitive)) -> (Output_Get:bool)
   Event OnParachuteOpened()
   > Switch on EGliderType(Selection=Get PlayerGliderType() -> (PlayerGliderType=HangGlider))
   > Set HangGlidingAlpha(HangGlidingAlpha:double=1) -> (Output_Get:double)
   Event OnParachuteClosed()
   > Set HangGlidingAlpha(HangGlidingAlpha=0)
  ]}
]
variables [
 Alpha_HeadRotateUpCrouchExtreme {type=number}
 AnvilAlarmMoth Detected {type=boolean}
 FE Idle {type=number}
 FE Idle Cons Value {type=number}
 Gravity Override {type=Vector}
 HangGlidingAlpha {type=number}
 RBAN Settings Out {type=FortRigidBodyAnimNodeParameters}
 Swimming {type=number}
]
